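# Sound cue for the final demolition impact on a snow wall, routed to the impacts sound class.
# A mixer layers four random picks: Yeti pickaxe impacts, briefly delayed snowball hits, holiday tree snow impacts and default rock final impacts, with the snowball layer boosted and the pickaxe and rock layers quieter.
# A top-level modulator varies pitch slightly per play, while small-object demolition attenuation and a four-voice concurrency limit control playback.
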SoundCue /Game/Sounds/Fort_Impact_Sounds/Demolition/Snow/Demo_SnowWall_ImpactFinal_Cue
root=Modulator_2
nodes [
 {id=Delay_0, children [Random_4], properties {delayMax=0.05, delayMin=0.05}}
 {id=Mixer_0, children [Random_6,Delay_0,Random_3,Random_0], properties {inputVolume [0.5,1.2,1,0.5]}}
 {id=Modulator_2, children [Mixer_0], properties {pitchMax=1.07, pitchMin=0.92}}
 {id=Random_0, children [WavePlayer_1,WavePlayer_2,WavePlayer_3], properties {weights [1,1,1]}}
 {id=Random_3, children [WavePlayer_11, WavePlayer_12, WavePlayer_13], properties {weights [1,1,1]}}
 {id=Random_4, children [WavePlayer_15, WavePlayer_16, WavePlayer_17, WavePlayer_18], properties {weights [1,1,1,1]}}
 {id=Random_6, children [WavePlayer_9, WavePlayer_10, WavePlayer_14], properties {weights [1,1,1]}}
 {id=WavePlayer_1, properties {soundWaveAssetPtr=/Game/Sounds/Fort_Impact_Sounds/Demolition/Defaults/Fort_Impact_RockFinal_01}}
 {id=WavePlayer_10, properties {soundWaveAssetPtr=/Game/Athena/Sounds/Weapons/PickAxes/Yeti/PA_Yeti_Impact_03}}
 {id=WavePlayer_11, properties {soundWaveAssetPtr=/Game/Sounds/Fort_Impact_Sounds/Demolition/Tree/Trees_Holiday_Winter_Snow_Impacts_2017-003}}
 {id=WavePlayer_12, properties {soundWaveAssetPtr=/Game/Sounds/Fort_Impact_Sounds/Demolition/Tree/Trees_Holiday_Winter_Snow_Impacts_2017-004}}
 {id=WavePlayer_13, properties {soundWaveAssetPtr=/Game/Sounds/Fort_Impact_Sounds/Demolition/Tree/Trees_Holiday_Winter_Snow_Impacts_2017-005}}
 {id=WavePlayer_14, properties {soundWaveAssetPtr=/Game/Athena/Sounds/Weapons/PickAxes/Yeti/PA_Yeti_Impact_04}}
 {id=WavePlayer_15, properties {soundWaveAssetPtr=/Game/Sounds/Fort_Weapons/Impacts/IMPCT_SnowBall_Hit_1}}
 {id=WavePlayer_16, properties {soundWaveAssetPtr=/Game/Sounds/Fort_Weapons/Impacts/IMPCT_SnowBall_Hit_2}}
 {id=WavePlayer_17, properties {soundWaveAssetPtr=/Game/Sounds/Fort_Weapons/Impacts/IMPCT_SnowBall_Hit_3}}
 {id=WavePlayer_18, properties {soundWaveAssetPtr=/Game/Sounds/Fort_Weapons/Impacts/IMPCT_SnowBall_Hit_4}}
 {id=WavePlayer_2, properties {soundWaveAssetPtr=/Game/Sounds/Fort_Impact_Sounds/Demolition/Defaults/Fort_Impact_RockFinal_02}}
 {id=WavePlayer_3, properties {soundWaveAssetPtr=/Game/Sounds/Fort_Impact_Sounds/Demolition/Defaults/Fort_Impact_RockFinal_03}}
 {id=WavePlayer_9, properties {soundWaveAssetPtr=/Game/Athena/Sounds/Weapons/PickAxes/Yeti/PA_Yeti_Impact_02}}
]
properties {
 attenuationSettings=/Game/Sounds/Attenuation/Fort_Demolition_Destroy_Small_Sized_Object_Att
 maxDistance=4000
 soundClassObject=&Fort_Impacts_PSM
 volumeMultiplier=0.68
 concurrencyOverrides {maxCount=4}
 concurrencySet [&CONC_Object_Impacts_FourMax]
}
